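; PCG graph that scatters single-tire meshes over WheelieHot tire mound actors and along splines.
; Mound surfaces are sampled through a class-filtered world ray query, then irregular-surface points are discarded, density-filtered and jittered.
; Splines split by TireSplineA and Rim tags into three distance-sampled branches with distinct random transforms; only Rim points get active self-pruning.
; All four branches merge into one static mesh spawner.
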(graph PCG_SwiftApple_TireScatter
  (bind WorldRayHitQuery_1 (World%20Ray%20Hit%20Query :QueryParams (struct :bIgnoreBackfaceHits true :bIgnorePCGHits true :ActorClassFilter Include :ActorClass /WildEstate/Environment/Props/WheelieHot/WheelieHot_TireMound_A/Blueprints/BP_WheelieTire_TireMound_A.BP_WheelieTire_TireMound_A_C :SelectLandscapeHits Exclude)))
  (bind SurfaceSampler_2 (Surface%20Sampler :PointExtents (3e+01 3e+01 3e+01) :PointsPerSquaredMeter 2.0))
  (bind Subgraph_0 (Subgraph :Graph /PCG/Subgraphs/DiscardPointsOnIrregularSurface.DiscardPointsOnIrregularSurface)) ; Seems to help remove too many floating tires
  (bind DensityFilter_0 (Density%20Filter :LowerBound 0.6696))
  (bind TransformPoints_1 (Transform%20Points :OffsetMax (35.0 35.0 0.0) :OffsetMin (-35.0 -35.0 0.0))) ; Scatter
  (bind TransformPoints_0 (Transform%20Points :OffsetMax (0.0 0.0 -1e+01) :OffsetMin (0.0 0.0 -1e+01) :RotationMax (struct :Pitch 3e+01 :Yaw 1.8e+02 :Roll 3e+01) :RotationMin (struct :Pitch -15.0 :Roll -3e+01) :ScaleMax (1.2 1.2 1.2) :ScaleMin (0.9 0.9 0.9)))
  (bind SelfPruning_1 (Self%20Pruning :Parameters (struct :PruningType None :ComparisonSource $Extents :RadiusSimilarityFactor 0.390713)))
  (bind GetSplineData_0 (Get%20Spline%20Data :ComponentSelector (struct :ComponentSelection ByClass :ComponentSelectionTag TireSpline :ComponentSelectionClass /Script/Engine.SplineComponent)))
  (bind FilterDataByTag_0 (Filter%20Data%20By%20Tag :SelectedTags "TireSplineA"))
  (bind FilterDataByTag_1 (Filter%20Data%20By%20Tag :SelectedTags "Rim"))
  (bind SplineSampler_1 (Spline%20Sampler :SamplerParams (struct :Mode Distance :SubdivisionsPerSegment 10 :DistanceIncrement 157.94412 :StartOffset 0.072 :MaxRandomOffsetNormalized 1.062242 :PointSteepness 1.0)))
  (bind TransformPoints_2 (Transform%20Points :OffsetMax (0.0 -1e+01 0.0) :OffsetMin (0.0 -2e+01 0.0) :RotationMax (struct :Pitch 19.05359 :Yaw 11.004087 :Roll 135.064441) :RotationMin (struct :Pitch -24.871763 :Yaw -8.238312 :Roll 55.979691)))
  (bind SplineSampler_0 (Spline%20Sampler :SamplerParams (struct :Mode Distance :SubdivisionsPerSegment 10 :DistanceIncrement 1.2e+02 :MaxRandomOffsetNormalized 1.062242 :PointSteepness 1.0)))
  (bind TransformPoints_3 (Transform%20Points :RotationMax (struct :Pitch 45.0 :Yaw 45.0 :Roll 45.0)))
  (bind SplineSampler_2 (Spline%20Sampler :SamplerParams (struct :Mode Distance :SubdivisionsPerSegment 10 :DistanceIncrement 157.94412 :StartOffset 0.072 :MaxRandomOffsetNormalized 1.062242 :bUnbounded true :PointSteepness 1.0)))
  (bind SelfPruning_0 (Self%20Pruning :Parameters (struct :PruningType SmallToLarge :ComparisonSource $Extents)))
  (bind TransformPoints_4 (Transform%20Points :OffsetMax (0.0 -1e+01 0.0) :OffsetMin (0.0 -2e+01 0.0) :RotationMax (struct :Pitch 1e+01 :Yaw 1.8e+02 :Roll 1e+01) :RotationMin (struct :Pitch -1e+01 :Yaw -1.8e+02 :Roll -1e+01)))
  (bind MergePoints_1 (Merge%20Points))
  (bind StaticMeshSpawner_1 (Static%20Mesh%20Spawner :MeshSelectorParameters (class:/Script/PCG.PCGMeshSelectorWeighted :MeshEntries [(struct :Descriptor (struct :StaticMesh /WildEstate/Environment/Props/WheelieHot/WheelieHot_TireMound_A/Meshes/SM_WheelieHot_SingleTire_A.SM_WheelieHot_SingleTire_A))])))
  (connect Subgraph_0 Points%20Out DensityFilter_0 In)
  (connect GetSplineData_0 Out FilterDataByTag_0 In)
  (connect FilterDataByTag_0 OutsideFilter FilterDataByTag_1 In)
  (connect SelfPruning_1 Out MergePoints_1 In)
  (connect TransformPoints_3 Out MergePoints_1 In2)
  (connect TransformPoints_4 Out MergePoints_1 In3)
  (connect TransformPoints_2 Out MergePoints_1 In4)
  (connect SplineSampler_2 Out SelfPruning_0 In)
  (connect TransformPoints_0 Out SelfPruning_1 In)
  (connect FilterDataByTag_0 InsideFilter SplineSampler_0 Spline)
  (connect FilterDataByTag_1 OutsideFilter SplineSampler_1 Spline)
  (connect FilterDataByTag_1 InsideFilter SplineSampler_2 Spline)
  (connect MergePoints_1 Out StaticMeshSpawner_1 In)
  (connect SurfaceSampler_2 Out Subgraph_0 Points%20In)
  (connect WorldRayHitQuery_1 Out Subgraph_0 Surface)
  (connect WorldRayHitQuery_1 Out SurfaceSampler_2 Surface)
  (connect TransformPoints_1 Out TransformPoints_0 In)
  (connect DensityFilter_0 Out TransformPoints_1 In)
  (connect SplineSampler_1 Out TransformPoints_2 In)
  (connect SplineSampler_0 Out TransformPoints_3 In)
  (connect SelfPruning_0 Out TransformPoints_4 In)
  (output Out StaticMeshSpawner_1))
